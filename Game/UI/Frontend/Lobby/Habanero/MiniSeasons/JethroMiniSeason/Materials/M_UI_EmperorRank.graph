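# Translucent UI material for the Emperor rank texture in the Jethro mini-season lobby UI, with animated highlight and shine.
# Widget and texture aspect ratios correct the UVs to fit width, with anchor and scale parameters.
# A delayed, ease-out timer drives an angled reflected-gradient "Highlight" and a rotated sine-band red "Shining Effect".
# A lerp blends the sampled color toward a saturated, doubled-brightness version.
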Material /Game/UI/Frontend/Lobby/Habanero/MiniSeasons/JethroMiniSeason/Materials/M_UI_EmperorRank
graph {
 properties {
  bAutomaticallySetUsageInEditor=false
  bUsedWithStaticMesh=false
  blendMode=BLEND_Translucent
  materialDomain=MD_UI
  usedShadingModels=Default Lit
 }
 nodes [
  Composite() {subgraphName=Fix widget aspect ratio}
  n1 = ScalarParameter() {defaultValue=1, group=Texture, parameterName=TextureScale}
  n2 = TextureCoordinate()
  n3 = AppendVector(
   A=ScalarParameter() {defaultValue=0.5, group=Texture, parameterName=Anchor X}
   B=ScalarParameter() {defaultValue=0.5, group=Texture, parameterName=Anchor Y}
  )
  n4 = MaterialFunctionCall() {materialFunction=&GetUserInterfaceUV}
  n5 = AppendVector(A=Constant() {r=1}, B=Divide(A=ComponentMask(None=@n4.Pixel Size) {g=true}, B=ComponentMask(None=@n4.Pixel Size) {r=true})) {desc=Widget Vertical Aspect Ratio}
  MaterialExpressionReroute_5 = Reroute(None=@n5)
  n6 = TextureObjectParameter() {parameterName=Texture, texture=/Game/UI/Frontend/Lobby/Habanero/MiniSeasons/JethroMiniSeason/Textures/T_UI_Emperor}
  n7 = TextureProperty(Tex=@n6) {property=TMTM_TexelSize}
  n8 = ComponentMask(None=@n7) {r=true}
  n9 = ComponentMask(None=@n7) {g=true}
  n10 = Divide(A=@MaterialExpressionReroute_5.Output, B=AppendVector(A=Constant() {r=1}, B=Divide(A=@n8, B=@n9)) {desc=Texture Vertical Aspect Ratio})
  n11 = ScalarParameter() {defaultValue=0.75, group=Texture, parameterName=UserScale}
  n12 = Add(A=Multiply(A=Subtract(A=@n2, B=@n3), B=Divide(A=@n10, B=@n11)) {desc="Scaling to fix Aspect Ratio, fit to width"}, B=@n3)
  n13 = TextureSample(
   UVs=MaterialFunctionCall!(
    Texture Scale=AppendVector(A=@n1, B=@n1)
    UVs=NamedRerouteDeclaration(None=@n12) {name=AdjustedTextureUV, nodeColor=<rgba 0.8984 1 0 1>}
    Offset XY Texturesize Z=Constant3Vector() {constant=<rgba 1 0 1 0>}
   ) {materialFunction=/Engine/Functions/Engine_MaterialFunctions01/Texturing/ScaleUVsAroundPoint}.UVs
   Tex=@n6
  ) {samplerSource=SSM_Clamp_WorldGroupSettings}
  n14 = Multiply(A=@n13.RGB) {constB=2}
  LinearInterpolate(A=@n13.RGB, B=Saturate(None=Max(A=@n13.RGB, B=@n14)), Alpha=NamedRerouteUsage())
  n15 = Constant() {r=2}
  n16 = Constant() {r=2}
  n17 = Subtract(A=MaterialFunctionCall("Value (0-1)"=MaterialFunctionCall!(Animation Duration=@n15, Delay=@n16, Reset to Default) {materialFunction=/FortUILibrary/MaterialFunctions/Time/MF_UI_AnimationDelay}.Result) {materialFunction=&MF_UI_EaseCurves}.EaseOutCubic) {constB=0.5}
  NamedRerouteDeclaration(None=MaterialFunctionCall!(UVs, Offset=Multiply(A=@n17) {constB=4}, RotationCenter, RotationAngle=Constant() {r=-0.125}) {materialFunction=/FortUILibrary/MaterialFunctions/Gradients/MF_UI_ReflectedGradient}.Result) {name=Highlight, nodeColor=<rgba 0 0.8594 1 1>}
  n18 = MaterialFunctionCall() {materialFunction=&GetUserInterfaceUV}
  n19 = Constant2Vector() {r=-0.5}
  n20 = Multiply(A=MaterialFunctionCall(UVs=Add(A=@n18.Normalized UV, B=@n19), "Rotation Angle (0-1)"=Constant() {r=-0.125}) {materialFunction=&CustomRotator}.Rotated Values) {constB=0.5}
  n21 = Frac(None=Subtract(A=Add(A=@n20), B=@n17))
  n22 = Constant() {r=0.1}
  MaterialExpressionRemap_4 = MaterialXRemap!(None=Subtract(A=Sine(None=@n21), B=OneMinus(None=@n22)) {constB=0.5}, InputLow, InputHigh=@n22, TargetLow, TargetHigh)
  n23 = ComponentMask(None=Saturate(None=@MaterialExpressionRemap_4)) {r=true}
  NamedRerouteDeclaration(None=Multiply(A=Multiply(A=@n23) {constB=4}, B=Constant3Vector() {constant=<rgba 1 0 0.0324 1>})) {name=Shining Effect, nodeColor=<rgba 0 0 0 1>}
  PinBase!(Output=@n5) {reroutePins [Output {expression=&self:MaterialExpressionReroute_5}]}
  PinBase!()
 ]
}
