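# Translucent UI material that plays a looping animated lightning effect, in a folder for the Jethro mini-season lobby screens.
# Looping time, normalized over a 5.2-second period, samples a curve atlas row that sets the phase of a FlipBook over a six-frame lightning texture.
# If nodes zero both color and alpha unless the curve value exceeds a small threshold.
# The "L or R" scalar blends to inverted UVs, flipping both axes; its phase blend has identical inputs.
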Material /Game/UI/Frontend/Lobby/Habanero/MiniSeasons/JethroMiniSeason/Materials/M_UI_JethroLightning
graph {
 properties {
  bAutomaticallySetUsageInEditor=false
  bUsedWithStaticMesh=false
  blendMode=BLEND_Translucent
  materialDomain=MD_UI
  usedShadingModels=Default Lit
 }
 nodes [
  n1 = CurveAtlasRowParameter(CurveTime=Divide(A=Time() {bOverride_Period=true, period=5.2}) {constB=5.2}) {
   atlas=/Game/UI/Frontend/Lobby/Habanero/MiniSeasons/JethroMiniSeason/Data/CA_JethroAnim
   curve=/Game/UI/Frontend/Lobby/Habanero/MiniSeasons/JethroMiniSeason/Data/CC_AnimLightningCurve
   parameterName=123
  }
  n2 = ScalarParameter() {defaultValue=1, parameterName=L or R}
  n3 = LinearInterpolate(A=@n1.R, B=@n1.R, Alpha=@n2)
  n4 = TextureCoordinate()
  n5 = MaterialFunctionCall(
   "Animation  Phase (0-1)"=@n3
   Number of Rows=Constant() {r=3}
   Number of Columns=Constant() {r=2}
   Texture=TextureObject() {texture=/Game/UI/Frontend/Lobby/Habanero/MiniSeasons/JethroMiniSeason/Textures/T_UI_EmperorLightningAtlas}
   UVs=LinearInterpolate(A=@n4, B=OneMinus(None=@n4), Alpha=@n2)
  ) {materialFunction=&FlipBook}
  n6 = Constant()
  If(A=@n3, "A > B"=@n5.Result, "A == B"=@n6, "A < B"=@n6) {constB=0.1}
  If(A=@n3, "A > B"=@n5.Alpha, "A == B"=@n6, "A < B"=@n6) {constB=0.1}
 ]
}
