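# Temporary material function in the Jethro mini-season lobby UI data that raises or lowers the saturation of a colored RGBA gradient within a masked area.
# The saturation input is clamped and negated to drive a Desaturation node with equal grey luminance factors.
# The mask, capped by a threshold through Min, defines where the adjusted color blends in, and the output is saturated RGBA.
# Unconnected bicolor and mirrored UV gradient nodes also remain in the graph.
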MaterialFunction /Game/UI/Frontend/Lobby/Habanero/MiniSeasons/JethroMiniSeason/Data/MF_Temporary_Gradient_TransitionSaturation
graph {nodes [
  Constant() {r=0.2}
  Constant() {r=0.95}
  n1 = NamedRerouteUsage()
  n2 = Desaturation(None=@n1, Fraction=NamedRerouteUsage()) {
   desc="Must be grey 0.5,0.5,0.5 to work with \r\na maximum of colors without having to \r\nchange the saturation value"
   luminanceFactors=<rgba 0.5 0.5 0.5 1>
  }
  n3 = LinearInterpolate(A=ComponentMask(None=@n1) {b=true, desc=Because Desaturation output a Float3, g=true, r=true}, B=@n2, Alpha=NamedRerouteUsage())
  n4 = ComponentMask(None=NamedRerouteUsage()) {a=true}
  FunctionOutput(None=Saturate(None=AppendVector(A=@n3, B=@n4))) {outputName=Colorful RGBA Gradient}
  n5 = TextureCoordinate()
  LinearInterpolate(
   A=Saturate(None=Constant4Vector() {constant=<rgba 1 0.6431 0 1>}.RGBA)
   B=Saturate(None=Constant4Vector() {constant=<rgba 0 0.1765 1 1>}.RGBA)
   Alpha=Saturate(None=ComponentMask(None=@n5) {r=true})
  ) {desc=classic bicolor gradient}
  n6 = FunctionInput() {
   bUsePreviewValueAsDefault=true
   description="Space where to apply the saturation (can be a gradient, a shape...)."
   inputName=Saturation Mask
   inputType=FunctionInput_Scalar
   sortPriority=4
  }
  n7 = FunctionInput() {
   bUsePreviewValueAsDefault=true
   description="Where to apply the saturation (can be a gradient, a shape...)."
   inputName=Saturation Mask Threshold
   inputType=FunctionInput_Scalar
   sortPriority=5
   previewValue=<0.95 0 0 1>
  }
  NamedRerouteDeclaration(None=Min(A=Saturate(None=@n6), B=Saturate(None=@n7))) {name=Saturation Space, nodeColor=<rgba 0 0 0 1>}
  n8 = FunctionInput() {bUsePreviewValueAsDefault=true, inputName="Saturation (-1 to 1)", inputType=FunctionInput_Scalar, sortPriority=3}
  NamedRerouteDeclaration(None=Multiply(A=Clamp(None=@n8) {maxDefault=2, minDefault=-2}) {constB=-1, desc=Because using desaturation node}) {name=Saturation, nodeColor=<rgba 0 0 0 1>}
  n9 = FunctionInput() {
   bUsePreviewValueAsDefault=true
   description=Can be any Colored Gradient.
   inputName=Colored Gradient
   inputType=FunctionInput_Vector4
   sortPriority=1
   previewValue=<0 0 0 0>
  }
  NamedRerouteDeclaration(None=@n9) {name=Colored Gradient Source, nodeColor=<rgba 0 0 0 1>}
  n10 = ComponentMask(None=TextureCoordinate()) {r=true}
  n11 = OneMinus(None=@n10)
  Subtract(A=Multiply(A=Min(A=@n10, B=@n11)) {constB=2}) {constB=0.25}
 ]}
