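# Translucent UI material for the Jethro mini-season Emperor background, combining a color gradient with animated smoky fog.
# Temporary gradient functions build the base color from parameter colors with saturation control.
# Panning noise distorts UVs, partly in screen space, while two scrolling cloud textures form smoke and radial and edge masks shape it.
# Fog color mixes navy, red and yellow tints, and scalar parameters control fog opacity, red fog and distortion.
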Material /Game/UI/Frontend/Lobby/Habanero/MiniSeasons/JethroMiniSeason/Materials/M_UI_Jethro_Emperor_Background
graph {
 properties {
  bUsedWithStaticMesh=false
  blendMode=BLEND_Translucent
  materialDomain=MD_UI
  pixelDepthOffsetMode=PDOM_Legacy
  usedShadingModels=Default Lit
 }
 nodes [
  n1 = VectorParameter() {parameterName=c21, defaultValue=<rgba 0.1248 0.0242 0.7529 1>}
  n2 = VectorParameter() {parameterName=c22, defaultValue=<rgba 0.8714 0.531 0.6123 1>}
  n3 = NamedRerouteUsage()
  n4 = LinearInterpolate(A=AppendVector(A=@n1.RGB, B=@n1.RGB), B=AppendVector(A=@n2.RGB, B=@n2.RGB), Alpha=@n3)
  n5 = VectorParameter() {parameterName=c2, defaultValue=<rgba 0.605 0.2534 0.9387 1>}
  n6 = AppendVector(A=@n5.RGB, B=@n5.RGB)
  n7 = VectorParameter() {parameterName=c1, defaultValue=<rgba 0.2159 0.0116 0.1746 1>}
  n8 = AppendVector(A=@n7.RGB, B=@n7.RGB)
  n9 = NamedRerouteUsage()
  LinearInterpolate(A=@n4, B=LinearInterpolate(A=@n6, B=@n8, Alpha=@n3), Alpha=@n9)
  n10 = SmoothStep(Value=Multiply(A=NamedRerouteUsage(), B=NamedRerouteUsage())) {constMax=0.7}
  LinearInterpolate(B=NamedRerouteUsage(), Alpha=@n10) {constB=0}
  n11 = MaterialFunctionCall() {materialFunction=&MF_UI_UVScaling}
  n12 = Constant() {r=1}
  n13 = TextureSample(UVs=Panner(Coordinate=MaterialFunctionCall(UVs=@n11.Scaled to Fit, Texture Scale=@n12) {materialFunction=&ScaleUVsByCenter}.UVs) {speedX=0.05, speedY=0.01}) {samplerType=SAMPLERTYPE_Grayscale, texture=/FortUILibrary/UIKit/Buttons/Textures/LowResBlurredNoise}
  n14 = MaterialFunctionCall() {materialFunction=&MF_UI_UVScaling}
  n15 = Constant() {r=1}
  n16 = TextureSample(UVs=Panner(Coordinate=MaterialFunctionCall(UVs=@n14.Scaled to Fit, Texture Scale=@n15) {materialFunction=&ScaleUVsByCenter}.UVs) {speedX=-0.02, speedY=0.04}) {samplerType=SAMPLERTYPE_Grayscale, texture=&LowResBlurredNoise}
  Max(A=@n13.G, B=@n16.G)
  n17 = Constant() {r=1}
  NamedRerouteDeclaration(None=MaterialFunctionCall!(
    Color A=@n6
    Color B=@n8
    Color C=MaterialFunctionCall!(Colored Gradient=@n4, "Saturation (-1 to 1)"=@n17, Saturation Mask=@n3, Saturation Mask Threshold) {materialFunction=/Game/UI/Frontend/Lobby/Habanero/MiniSeasons/JethroMiniSeason/Data/MF_Temporary_Gradient_TransitionSaturation}.Colorful RGBA Gradient
    A to B Gradient=@n3
    Color C Space=@n9
    "Display Color C (0 to 1)"=@n17
    "Saturation (-1 to 1)"
    Saturation Mask
    Saturation Mask Threshold
    "Color C Covers Saturation (0 to 1)"
   ) {materialFunction=/Game/UI/Frontend/Lobby/Habanero/MiniSeasons/JethroMiniSeason/Data/MF_Temporary_Gradient_3Colors_WithSaturation}.Colorful RGBA Gradient) {name=BaseColor, nodeColor=<rgba 0.6172 1 0 1>}
  n18 = MaterialFunctionCall() {materialFunction=&GetUserInterfaceUV}
  n19 = Multiply(A=Power(Base=Max(A=@n13.R, B=@n16.R))) {constB=0.12}
  n20 = Multiply(A=Power(Base=Min(A=@n13.R, B=@n16.R)) {constExponent=3}) {constB=0.05}
  NamedRerouteDeclaration(None=Add(A=NamedRerouteDeclaration(None=@n18.Normalized UV) {name=Normalized UV, nodeColor=<rgba 0 0 0 1>}, B=AppendVector(A=@n19, B=@n20))) {name=Distorted UV, nodeColor=<rgba 0 0 0 1>}
  n21 = Panner(
   Coordinate=MaterialFunctionCall(
    UVs=MaterialFunctionCall() {materialFunction=&MF_UI_UVScaling}.Scaled to Fit
    Texture Scale=Constant() {r=0.3}
   ) {materialFunction=&ScaleUVsByCenter}.UVs
   Time=NamedRerouteUsage()
  ) {speedX=0.012, speedY=0.05}
  NamedRerouteDeclaration(None=Multiply(A=TextureSample(UVs=@n21) {texture=/Game/Environments/FrontEnd/Textures/Noise}.R, B=Constant() {r=0.5})) {name=Noise, nodeColor=<rgba 0 0 0 1>}
  n22 = MaterialFunctionCall() {materialFunction=&MF_UI_UVScaling}
  n23 = MaterialFunctionCall(UVs=ScreenPosition().ViewportUV, Texture Scale=Constant() {r=0.1}) {materialFunction=&ScaleUVsByCenter}
  n24 = Multiply(A=Time(), B=Constant() {r=0.02})
  n25 = ScalarParameter() {defaultValue=0.03, parameterName=Distort Amount}
  n26 = Multiply(
   A=TextureSample(UVs=Panner(Coordinate=@n23.UVs, Time=@n24) {speedX=-0.5, speedY=0.4}) {samplerType=SAMPLERTYPE_Grayscale, texture=/FortUILibrary/UGC/Textures/T_UEFN_LowResBlurredNoise}.R
   B=@n25
  ) {constB=0.2}
  n27 = ScreenPosition()
  n28 = Constant() {r=0.2}
  n29 = TextureSample(UVs=Panner(
    Coordinate=MaterialFunctionCall(UVs=@n27.ViewportUV, Texture Scale=@n28) {materialFunction=&ScaleUVsByCenter}.UVs
    Time=@n24
   ) {speedX=-0.4, speedY=0.4}) {samplerType=SAMPLERTYPE_Grayscale, texture=&T_UEFN_LowResBlurredNoise}
  NamedRerouteDeclaration(None=Add(A=Add(A=@n22.Scaled to Fit, B=@n26), B=Multiply(A=@n29.R, B=@n25) {constB=0.2})) {name=Smoke Distorted UV, nodeColor=<rgba 0 0 0 1>}
  n30 = NamedRerouteUsage()
  n31 = Add(A=TextureCoordinate(), B=Constant2Vector() {g=0.175})
  n32 = Multiply(
   A=SmoothStep(Value=ComponentMask(None=@n31) {g=true}) {constMin=0.25}
   B=MaterialFunctionCall(UV=@n31) {materialFunction=&MF_UI_RadialGradient}.Smooth
  )
  n33 = Multiply(A=NamedRerouteUsage(), B=NamedRerouteUsage())
  n34 = Add(A=Saturate(None=Multiply(A=@n32) {constB=1.8}), B=SquareRoot(None=SmoothStep(Value=@n33) {constMax=0.8, constMin=0.1}))
  n35 = Multiply(A=NamedRerouteUsage()) {constB=0.5}
  n36 = Add(A=Saturate(None=Subtract(A=@n34, B=@n35)), B=@n10)
  n37 = LinearInterpolate(
   B=Multiply(A=Saturate(None=@n36), B=ScalarParameter() {defaultValue=0.5, parameterName=FogMainOpacity})
   Alpha=ScalarParameter() {defaultValue=1, parameterName=UseFog}
  )
  NamedRerouteDeclaration(None=LinearInterpolate(A=Multiply(A=@n30, B=@n37), B=@n37, Alpha=ScalarParameter() {defaultValue=1, parameterName=UseRedFog})) {name=FogAlpha, nodeColor=<rgba 0.125 1 0 1>}
  n38 = NamedRerouteUsage()
  n39 = OneMinus(None=Distance(A=ComponentMask(None=@n38) {r=true}, B=Constant() {r=0.5}))
  n40 = OneMinus(None=ComponentMask(None=@n38) {g=true})
  n41 = OneMinus(None=Distance(A=ComponentMask(None=@n38) {r=true}, B=Constant() {r=0.9096}))
  n42 = ComponentMask(None=@n38) {g=true}
  n43 = Max(
   A=Multiply(A=SmoothStep(Value=@n39) {constMin=0.82}, B=SmoothStep(Value=@n40) {constMax=0.65, constMin=0.35})
   B=Multiply(A=SmoothStep(Value=@n41) {constMin=0.82}, B=SmoothStep(Value=@n42) {constMax=0.65, constMin=0.35})
  )
  n44 = Constant() {r=1}
  n45 = LinearInterpolate(
   A=AppendVector(A=Constant3Vector() {constant=<rgba 0 0.0188 0.1198 1>}, B=@n44)
   B=AppendVector(A=Constant3Vector() {constant=<rgba 0.6795 0 0.0726 1>}, B=@n44)
   Alpha=NamedRerouteUsage()
  )
  n46 = Add(A=NamedRerouteUsage(), B=Multiply(A=Multiply(A=@n43, B=@n45), B=Constant() {r=0.3}) {constB=2})
  n47 = AppendVector(A=Constant3Vector() {constant=<rgba 1 0.9029 0 1>}, B=Constant() {r=1})
  n48 = NamedRerouteUsage()
  n49 = Multiply(A=NamedRerouteUsage(), B=NamedRerouteUsage())
  n50 = Multiply(A=SmoothStep(Value=Multiply(A=@n48, B=@n49)) {constMax=0.7}, B=Constant() {r=0.22}) {constB=0.25}
  NamedRerouteDeclaration(None=LinearInterpolate(B=LinearInterpolate(A=@n46, B=@n47, Alpha=@n50), Alpha=ScalarParameter() {defaultValue=1, parameterName=UseFog})) {name=FogColor, nodeColor=<rgba 1 0 0.4453 1>}
  n51 = MaterialFunctionCall!(Vector or UVs=NamedRerouteUsage()) {materialFunction=/FortUILibrary/MaterialFunctions/Basics/MF_UI_CheaperVectorToRadial}
  n52 = MaterialXRemap!(None=Cosine(None=Multiply(A=@n51.Vector Converted to Angle) {constB=2}), InputLow, InputHigh, TargetLow, TargetHigh)
  n53 = NamedRerouteUsage()
  NamedRerouteDeclaration(None=Saturate(None=Max(A=@n52, B=@n53))) {name=FogColorDivider, nodeColor=<rgba 0 0.1094 1 1>}
  n54 = Constant() {r=2}
  n55 = Constant() {r=1.8}
  n56 = Multiply(A=NamedRerouteUsage(), B=Constant() {r=0.8})
  n57 = Time()
  n58 = Constant() {r=0.5}
  n59 = Panner(
   Coordinate=Multiply(A=@n56, B=AppendVector(A=@n54, B=@n55))
   Time=NamedRerouteDeclaration(None=Multiply(A=@n57, B=@n58)) {name=Smoke_Bkg_Time, nodeColor=<rgba 0 0 0 1>}
  ) {bFractionalPart=true, speedX=-0.02, speedY=0.04}
  n60 = Saturate(None=Add(A=TextureSample(UVs=@n59) {samplerType=SAMPLERTYPE_LinearGrayscale, texture=/Game/Global/Textures/Default/Tileable_Patterns/T_StormCloudDetail01_M}.G) {constB=0.2})
  n61 = Multiply(A=@n56, B=AppendVector(A=Constant() {r=1.5}, B=Constant() {r=1.3}))
  n62 = NamedRerouteUsage()
  n63 = TextureSample(UVs=NamedRerouteDeclaration(None=Panner(Coordinate=@n61, Time=@n62) {bFractionalPart=true, speedX=-0.04, speedY=0.02}) {name=Smoke_Bkg_AnimUvs, nodeColor=<rgba 0 0 0 1>}) {texture=/Game/Athena/SafeZone/S11/T_TilingClouds_D}
  NamedRerouteDeclaration(None=Multiply(A=LinearInterpolate(A=@n60, B=@n63.G), B=Constant() {r=0.8})) {name=Smoke Shape, nodeColor=<rgba 0 0 0 1>}
  n64 = NamedRerouteUsage()
  n65 = Distance(A=ComponentMask(None=@n64) {r=true}, B=Constant() {r=0.5})
  n66 = SmoothStep(Value=OneMinus(None=Multiply(A=@n65) {constB=2})) {constMax=0.4, constMin=-0.1}
  n67 = Constant()
  n68 = SmoothStep(Min=@n67, Value=OneMinus(None=ComponentMask(None=@n64) {g=true})) {constMax=0.5}
  n69 = ComponentMask(None=@n64) {g=true}
  n70 = Multiply(A=Multiply(A=Power(Base=@n66), B=Power(Base=@n68)), B=Power(Base=SmoothStep(Min=@n67, Value=@n69) {constMax=0.6}))
  n71 = NamedRerouteUsage()
  n72 = Distance(A=ComponentMask(None=@n71) {r=true}, B=Constant() {r=0.5})
  n73 = SmoothStep(Value=OneMinus(None=Multiply(A=@n72) {constB=2})) {constMax=0.9, constMin=-0.1}
  n74 = Constant() {r=-0.1}
  n75 = SmoothStep(Min=@n74, Value=ComponentMask(None=@n71) {g=true}) {constMax=0.2, constMin=-0.1}
  n76 = OneMinus(None=ComponentMask(None=@n71) {g=true})
  n77 = Multiply(A=Multiply(A=Power(Base=@n73), B=Power(Base=@n75)), B=Power(Base=SmoothStep(Min=@n74, Value=@n76) {constMax=0.2, constMin=-0.1}))
  n78 = TextureCoordinate()
  n79 = Constant2Vector() {g=0.5, r=0.5}
  n80 = MaterialFunctionCall(UV=Add(
    A=MaterialFunctionCall(UVs=@n78, Texture Scale=@n79) {materialFunction=&ScaleUVsByCenter}.UVs
    B=Constant2Vector()
   )) {materialFunction=&MF_UI_RadialGradient}
  n81 = OneMinus(None=Multiply(A=NamedRerouteDeclaration(None=@n80.Smooth) {name=Sphere, nodeColor=<rgba 0 0.2969 1 1>}, B=Constant() {r=0.5}))
  NamedRerouteDeclaration(None=Multiply(A=Max(A=@n70, B=@n77), B=SmoothStep(Value=@n81) {constMin=0.4})) {name=Shape, nodeColor=<rgba 0 0 0 1>}
 ]
}
