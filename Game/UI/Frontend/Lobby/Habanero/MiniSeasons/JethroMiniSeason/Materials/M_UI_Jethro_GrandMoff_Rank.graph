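# Translucent UI material for the Jethro mini-season Grand Moff rank graphic, adding animated shine and glare over a parameterized rank texture.
# A light-map texture supplies light-area and button masks, and named reroutes carry each effect layer.
# Delayed, eased timers drive a diagonal highlight sweep, shine and flicker, with a custom overlay blend.
# Radial gradients form glare strips and glints, while two collapsed composites build star flares pulsed by Flare Strength and Flickering Speed.
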Material /Game/UI/Frontend/Lobby/Habanero/MiniSeasons/JethroMiniSeason/Materials/M_UI_Jethro_GrandMoff_Rank
graph {
 properties {
  bAutomaticallySetUsageInEditor=false
  bUsedWithStaticMesh=false
  blendMode=BLEND_Translucent
  materialDomain=MD_UI
  usedShadingModels=Default Lit
 }
 nodes [
  n1 = NamedRerouteUsage()
  n2 = NamedRerouteUsage()
  Add(A=Max(A=Max(A=@n1, B=@n2), B=NamedRerouteUsage()), B=NamedRerouteUsage())
  n3 = Multiply(A=NamedRerouteUsage(), B=Constant3Vector() {constant=<rgba 0.1747 1 0.9292 1>})
  n4 = NamedRerouteUsage()
  n5 = LinearInterpolate(B=@n4, Alpha=Multiply(A=NamedRerouteUsage(), B=@n4))
  n6 = Add(A=NamedRerouteUsage(), B=LinearInterpolate(A=Add(A=@n5, B=@n4), B=NamedRerouteUsage(), Alpha=NamedRerouteUsage()))
  Add(A=LinearInterpolate(A=@n3, B=Add(A=@n3, B=@n6), Alpha=NamedRerouteUsage()), B=NamedRerouteUsage())
  n7 = MaterialFunctionCall() {materialFunction=&GetUserInterfaceUV}
  n8 = Constant2Vector() {g=0.2, r=0.18}
  Composite!(InputPin=MaterialFunctionCall(UVs=Add(A=@n7.Normalized UV, B=@n8), Texture Scale=Constant2Vector() {g=0.15, r=0.1}) {materialFunction=&ScaleUVsByCenter}.UVs)
  n9 = MaterialFunctionCall(
   UVs=Add(
    A=MaterialFunctionCall() {materialFunction=&GetUserInterfaceUV}.Normalized UV
    B=Constant2Vector() {g=-0.03, r=-0.26}
   )
   Texture Scale=Constant2Vector() {g=0.05, r=0.1}
  ) {materialFunction=&ScaleUVsByCenter}
  Composite(InputPin=@n9.UVs)
  n10 = TextureSampleParameter2D() {parameterName=Texture, texture=/Game/UI/Frontend/Lobby/Habanero/MiniSeasons/JethroMiniSeason/Textures/T_UI_Jethro_Rank_GrandMoff}
  NamedRerouteDeclaration(None=@n10.RGB) {name=RankTexture, nodeColor=<rgba 1 0.6094 0 1>}
  n11 = TextureSample() {texture=/Game/UI/Frontend/Lobby/Habanero/MiniSeasons/JethroMiniSeason/Textures/T_UI_Jethro_GrandMoff_LightMap}
  NamedRerouteDeclaration(None=@n11.G) {name=LightAreas, nodeColor=<rgba 0.2891 0 1 1>}
  NamedRerouteDeclaration(None=@n10.A) {name=TextureAlpha, nodeColor=<rgba 1 0.8672 0 1>}
  n12 = Constant() {r=2}
  n13 = Constant() {r=2}
  n14 = Subtract(A=MaterialFunctionCall("Value (0-1)"=MaterialFunctionCall!(Animation Duration=@n12, Delay=@n13, Reset to Default) {materialFunction=/FortUILibrary/MaterialFunctions/Time/MF_UI_AnimationDelay}.Result) {materialFunction=&MF_UI_EaseCurves}.EaseOutCubic) {constB=0.5}
  NamedRerouteDeclaration(None=MaterialFunctionCall!(UVs, Offset=Multiply(A=@n14) {constB=4}, RotationCenter, RotationAngle=Constant() {r=-0.125}) {materialFunction=/FortUILibrary/MaterialFunctions/Gradients/MF_UI_ReflectedGradient}.Result) {name=Highlight, nodeColor=<rgba 0 0.8594 1 1>}
  n15 = MaterialFunctionCall(Float2=MaterialFunctionCall() {materialFunction=&GetUserInterfaceUV}.Normalized UV) {materialFunction=&BreakOutFloat2Components}
  n16 = MaterialFunctionCall(X=Add(A=NamedRerouteUsage(), B=@n15.R), Y=@n15.G) {materialFunction=&MakeFloat2}
  MaterialExpressionReroute_3 = Reroute(None=@n16.Result)
  n17 = Constant2Vector() {g=0.23, r=-0.9}
  n18 = MaterialFunctionCall(
   UVs=MaterialFunctionCall(UVs=Add(A=@MaterialExpressionReroute_3, B=@n17), Texture Scale=Constant2Vector() {g=0.3768, r=0.6176}) {materialFunction=&ScaleUVsByCenter}.UVs
   Density=Constant() {r=1.5}
  ) {materialFunction=&RadialGradientExponential}
  n19 = NamedRerouteUsage()
  NamedRerouteDeclaration(None=Add(A=Multiply(A=@n18.RadialGradientExponential, B=@n19), B=NamedRerouteUsage())) {name=Blick, nodeColor=<rgba 1 0.8203 0 1>}
  n20 = Add(A=@n16.Result, B=Constant2Vector() {g=0.24, r=-0.9})
  n21 = Constant2Vector() {g=0.0692, r=0.7024}
  n22 = Add(A=@n16.Result, B=Constant2Vector() {g=0.245, r=-0.9})
  n23 = Constant2Vector() {g=0.1, r=0.2792}
  n24 = Max(
   A=MaterialFunctionCall(
    UVs=MaterialFunctionCall(UVs=@n20, Texture Scale=@n21) {materialFunction=&ScaleUVsByCenter}.UVs
    Density=Constant() {r=0.5}
   ) {materialFunction=&RadialGradientExponential}.RadialGradientExponential
   B=MaterialFunctionCall(
    UVs=MaterialFunctionCall(UVs=@n22, Texture Scale=@n23) {materialFunction=&ScaleUVsByCenter}.UVs
    Density=Constant() {r=0.5}
   ) {materialFunction=&RadialGradientExponential}.RadialGradientExponential
  )
  NamedRerouteDeclaration(None=Multiply(A=@n19, B=Multiply(A=@n24) {constB=6})) {name=CenterBlick, nodeColor=<rgba 0.5 0 1 1>}
  n25 = MaterialFunctionCall("Value (0-1)"=MaterialFunctionCall(Animation Duration=Constant() {r=2}, Delay=Constant() {r=2}) {materialFunction=&MF_UI_AnimationDelay}.Result) {materialFunction=&MF_UI_EaseCurves}
  NamedRerouteDeclaration(None=OneMinus(None=Subtract(A=@n25.EaseOutCubic) {constB=0.5})) {name=GlarePadding, nodeColor=<rgba 0 1 0.8359 1>}
  n26 = NamedRerouteUsage()
  MaterialExpressionNamedRerouteUsage_0 = NamedRerouteUsage()
  MaterialExpressionConstant_0 = Constant() {r=1}
  n27 = Custom!(BlendColor=@MaterialExpressionNamedRerouteUsage_0, BaseColor=@MaterialExpressionConstant_0) {
   code="float3 result = (BlendColor < 0.5) \r\n        ? (2.0 * BaseColor * BlendColor) \r\n        : (1.0 - 2.0 * (1.0 - BaseColor) * (1.0 - BlendColor));\r\nreturn result;"
   inputs [
    {
     inputName=BlendColor
     input {expression=&self:MaterialExpressionNamedRerouteUsage_0}
    }
    {
     inputName=BaseColor
     input {expression=&self:MaterialExpressionConstant_0}
    }
   ]
  }
  n28 = NamedRerouteUsage()
  n29 = NamedRerouteUsage()
  NamedRerouteDeclaration(None=Max(A=LinearInterpolate(A=@n26, B=@n27, Alpha=@n28), B=Multiply(A=@n29) {constB=0.05})) {name=HighlightEffect, nodeColor=<rgba 0.0312 1 0 1>}
  n30 = Sine(None=Multiply(A=@n25.EaseOutCubic) {constB=0.5})
  n31 = MaterialFunctionCall() {materialFunction=&MF_UI_ReflectedGradient}
  NamedRerouteDeclaration(None=Multiply(A=Saturate(None=@n30), B=Power(Base=@n31.Result) {constExponent=0.5})) {name=Flickering, nodeColor=<rgba 1 0.5156 0 1>}
  n32 = MaterialFunctionCall() {materialFunction=&GetUserInterfaceUV}
  n33 = Constant2Vector() {g=0.1618}
  n34 = MaterialFunctionCall(
   UVs=MaterialFunctionCall(UVs=Add(A=@n32.Normalized UV, B=@n33), Texture Scale=Constant2Vector() {g=0.1, r=1.6}) {materialFunction=&ScaleUVsByCenter}.UVs
   Density=Constant() {r=1}
  ) {materialFunction=&RadialGradientExponential}
  n35 = Constant2Vector() {g=0.2322}
  n36 = MaterialFunctionCall(
   UVs=MaterialFunctionCall(UVs=Add(A=@n32.Normalized UV, B=@n35), Texture Scale=Constant2Vector() {g=0.1, r=1.6}) {materialFunction=&ScaleUVsByCenter}.UVs
   Density=Constant() {r=1}
  ) {materialFunction=&RadialGradientExponential}
  NamedRerouteDeclaration(None=Multiply(A=Max(A=@n34.RadialGradientExponential, B=@n36.RadialGradientExponential), B=NamedRerouteUsage())) {name=StripsOfGlare, nodeColor=<rgba 0.4062 0 1 1>}
  NamedRerouteDeclaration(None=@n11.R) {name=ButtonsArea, nodeColor=<rgba 0.6172 1 0 1>}
  n37 = MaterialFunctionCall() {materialFunction=&GetUserInterfaceUV}
  n38 = Constant2Vector() {r=-0.5}
  n39 = Multiply(A=MaterialFunctionCall(UVs=Add(A=@n37.Normalized UV, B=@n38), "Rotation Angle (0-1)"=Constant() {r=-0.125}) {materialFunction=&CustomRotator}.Rotated Values) {constB=0.5}
  n40 = Frac(None=Subtract(A=Add(A=@n39), B=@n14))
  n41 = Constant() {r=0.1}
  MaterialExpressionReroute_2 = Reroute(None=@n41)
  MaterialExpressionRemap_4 = MaterialXRemap!(
   None=Subtract(A=Sine(None=@n40), B=OneMinus(None=@n41)) {constB=0.5}
   InputLow
   InputHigh=@MaterialExpressionReroute_2
   TargetLow
   TargetHigh
  )
  n42 = ComponentMask(None=Saturate(None=@MaterialExpressionRemap_4)) {r=true}
  NamedRerouteDeclaration(None=Multiply(A=Multiply(A=@n42) {constB=4}, B=Constant3Vector() {constant=<rgba 0 0.785 1 1>})) {name=Shining Effect, nodeColor=<rgba 0 0 0 1>}
  MaterialExpressionReroute_23 = Reroute(None=@n9.UVs)
  n43 = Multiply(A=Abs(None=Subtract(A=@MaterialExpressionReroute_23.InputPin) {constB=0.5})) {constB=2}
  n44 = ComponentMask(None=@n43) {r=true}
  n45 = ComponentMask(None=@n43) {g=true}
  n46 = ScalarParameter() {defaultValue=60, parameterName=Flare Strength}
  n47 = Time()
  n48 = ScalarParameter() {defaultValue=0.1, parameterName=Flickering Speed}
  n49 = MaterialFunctionCall("Value (0-1)"=Frac(None=Multiply(A=@n47, B=@n48))) {materialFunction=&MF_UI_EaseCurves}
  n50 = Min(A=@n49.EaseInOutSine, B=OneMinus(None=@n49.EaseInOutSine))
  n51 = Power(Base=OneMinus(None=Multiply(A=@n44, B=@n45)), Exp=LinearInterpolate(A=@n46, B=Add(A=@n46) {constB=-16}, Alpha=Multiply(A=@n50) {constB=2})) {constExponent=120}
  n52 = Saturate(None=Multiply(
    A=Saturate(None=@n51)
    B=MaterialFunctionCall(UV=@MaterialExpressionReroute_23.InputPin) {materialFunction=&MF_UI_RadialGradient}.Smooth
   ))
  MaterialExpressionReroute_24 = Reroute(None=@n52)
  NamedRerouteDeclaration(None=Multiply(A=Max(A=@MaterialExpressionReroute_24.Output, B=@MaterialExpressionReroute_3.Output), B=NamedRerouteUsage())) {name=Star, nodeColor=<rgba 1 0 0.3047 1>}
  PinBase() {reroutePins [InputPin {expression=&self:MaterialGraph_0.MaterialGraphNode_Composite_2.Collapsed Nodes.MaterialExpressionReroute_2}]}
  n53 = MaterialFunctionCall(UV=@MaterialExpressionReroute_2.InputPin) {materialFunction=&MF_UI_RadialGradient}
  n54 = Multiply(A=Abs(None=Subtract(A=@MaterialExpressionReroute_2.InputPin) {constB=0.5})) {constB=2}
  n55 = Multiply(A=ComponentMask(None=@n54) {r=true}, B=ComponentMask(None=@n54) {g=true})
  n56 = ScalarParameter() {defaultValue=60, parameterName=Flare Strength}
  n57 = Add(A=@n56) {constB=-16}
  n58 = Time()
  n59 = ScalarParameter() {defaultValue=0.1, parameterName=Flickering Speed}
  n60 = MaterialFunctionCall("Value (0-1)"=Frac(None=Multiply(A=@n58, B=@n59))) {materialFunction=&MF_UI_EaseCurves}
  n61 = Multiply(A=Min(A=@n60.EaseInOutSine, B=OneMinus(None=@n60.EaseInOutSine))) {constB=2}
  n62 = Saturate(None=Power(Base=OneMinus(None=@n55), Exp=LinearInterpolate(A=@n56, B=@n57, Alpha=@n61)) {constExponent=120})
  PinBase!(Output=Saturate(None=Multiply(A=@n62, B=@n53.Smooth))) {reroutePins [Output {expression=&self:MaterialGraph_0.MaterialGraphNode_Composite_2.Collapsed Nodes.MaterialExpressionReroute_3}]}
  PinBase!() {reroutePins [InputPin {expression=&self:MaterialExpressionReroute_23}]}
  PinBase!(Output=@n52) {reroutePins [Output {expression=&self:MaterialExpressionReroute_24}]}
 ]
}
