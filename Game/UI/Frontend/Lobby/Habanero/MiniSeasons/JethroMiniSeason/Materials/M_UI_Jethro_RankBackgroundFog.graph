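# Translucent UI material that draws animated fog behind the rank display in the Jethro mini-season lobby.
# Panning storm-cloud and noise textures form the smoke, while procedural smoothstep and radial-gradient masks shape the fog.
# Scrolling noise distorts the UVs, partly sampled in screen-space so the warp follows the viewport.
# Color blends between two colors with added tinted highlights, and opacity combines the masks under a main opacity parameter.
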Material /Game/UI/Frontend/Lobby/Habanero/MiniSeasons/JethroMiniSeason/Materials/M_UI_Jethro_RankBackgroundFog
graph {
 properties {
  bAutomaticallySetUsageInEditor=false
  bUsedWithStaticMesh=false
  blendMode=BLEND_Translucent
  materialDomain=MD_UI
  usedShadingModels=Default Lit
 }
 nodes [
  n1 = ComponentMask(None=NamedRerouteUsage()) {g=true}
  n2 = Multiply(A=SmoothStep(Value=NamedRerouteUsage()), B=SmoothStep(Value=OneMinus(None=@n1)) {constMax=0.8, constMin=0.25})
  n3 = NamedRerouteUsage()
  n4 = Distance(A=ComponentMask(None=@n3) {r=true}, B=Constant() {r=0.5})
  n5 = ComponentMask(None=@n3) {g=true}
  n6 = Multiply(A=SmoothStep(Value=OneMinus(None=@n4)) {constMin=0.82}, B=SmoothStep(Value=OneMinus(None=@n5)) {constMax=0.65, constMin=0.35})
  n7 = VectorParameter() {parameterName=Fog Highlights, defaultValue=<rgba 0 0.6486 0.6795 1>}
  Add(
   A=LinearInterpolate(A=NamedRerouteUsage(), B=NamedRerouteUsage(), Alpha=Power(Base=@n2) {constExponent=1.2})
   B=Multiply(A=Multiply(A=@n6, B=@n7.RGB), B=ScalarParameter() {defaultValue=1, parameterName=Highlights Intensity}) {constB=2}
  )
  n8 = Add(A=NamedRerouteUsage(), B=Constant2Vector() {g=0.175})
  n9 = Multiply(
   A=SmoothStep(Value=ComponentMask(None=@n8) {g=true}) {constMin=0.25}
   B=MaterialFunctionCall(UV=@n8) {materialFunction=&MF_UI_RadialGradient}.Smooth
  )
  n10 = Multiply(A=NamedRerouteUsage(), B=NamedRerouteUsage())
  n11 = Add(A=Saturate(None=Multiply(A=@n9) {constB=1.8}), B=SquareRoot(None=SmoothStep(Value=@n10) {constMax=0.8, constMin=0.1}))
  Multiply(A=Saturate(None=Saturate(None=@n11)), B=ScalarParameter() {defaultValue=0.95, parameterName=Main Opacity})
  n12 = NamedRerouteUsage()
  n13 = ComponentMask(None=@n12) {g=true}
  n14 = Power(Base=SmoothStep(Value=OneMinus(None=@n13)) {constMax=0.6, constMin=-0.15})
  n15 = Power(Base=SmoothStep(Value=ComponentMask(None=@n12) {g=true}) {constMax=0.6})
  Multiply(A=@n14, B=@n15)
  n16 = Distance(A=ComponentMask(None=@n12) {r=true}, B=Constant() {r=0.55})
  n17 = SmoothStep(Value=OneMinus(None=Multiply(A=@n16) {constB=1.7897})) {constMax=0.928}
  n18 = Multiply(A=Multiply(A=Power(Base=@n17) {constExponent=0.7}, B=@n14), B=@n15)
  n19 = MaterialFunctionCall(UVs=NamedRerouteUsage(), Texture Scale=Constant2Vector() {g=2, r=0.5}) {materialFunction=&ScaleUVsByCenter}
  n20 = Constant2Vector() {g=0.35}
  n21 = Multiply(
   A=MaterialFunctionCall(UV=Add(A=@n19.UVs, B=@n20)) {materialFunction=&MF_UI_RadialGradient}.Smooth
   B=ScalarParameter() {defaultValue=1, parameterName=Top Radial Mask Intensity}
  )
  Multiply(A=@n18, B=SmoothStep(Value=OneMinus(None=@n21)) {constMin=0.4})
  NamedRerouteDeclaration(None=@n18) {name=Shape, nodeColor=<rgba 0 0 0 1>}
  n22 = MaterialFunctionCall() {materialFunction=&MF_UI_UVScaling}
  n23 = MaterialFunctionCall(UVs=ScreenPosition().ViewportUV, Texture Scale=ScalarParameter() {defaultValue=0.1, parameterName=Distort Scale 1}) {materialFunction=&ScaleUVsByCenter}
  n24 = Multiply(A=Time(), B=ScalarParameter() {defaultValue=0.02, parameterName=Distort Speed})
  n25 = ScalarParameter() {defaultValue=0.03, parameterName=Distort Amount}
  n26 = Multiply(
   A=TextureSample(UVs=Panner(Coordinate=@n23.UVs, Time=@n24) {speedX=-0.5, speedY=0.4}) {texture=/Game/UI/Frontend/Lobby/Habanero/MiniSeasons/JethroMiniSeason/Textures/T_UI_SmokeEffectNoise}.G
   B=@n25
  ) {constB=0.2}
  n27 = ScreenPosition()
  n28 = ScalarParameter() {defaultValue=0.2, parameterName=Distort Scale 2}
  n29 = TextureSample(UVs=Panner(
    Coordinate=MaterialFunctionCall(UVs=@n27.ViewportUV, Texture Scale=@n28) {materialFunction=&ScaleUVsByCenter}.UVs
    Time=@n24
   ) {speedX=-0.4, speedY=0.4}) {texture=&T_UI_SmokeEffectNoise}
  NamedRerouteDeclaration(None=Add(A=Add(A=@n22.Scaled to Fit, B=@n26), B=Multiply(A=@n29.G, B=@n25) {constB=0.2})) {name=Smoke Distorted UV, nodeColor=<rgba 0 0 0 1>}
  n30 = AppendVector(A=Constant() {r=2}, B=Constant() {r=1.8})
  n31 = Multiply(A=NamedRerouteUsage(), B=ScalarParameter() {defaultValue=0.8, group=Smoke1, parameterName=Smoke_Scale})
  n32 = Multiply(A=Time(), B=ScalarParameter() {defaultValue=0.5, group=Smoke1, parameterName=Bg_Smoke_TimeMultiply})
  n33 = TextureSample(UVs=Panner(Coordinate=Multiply(A=@n31, B=@n30), Time=NamedRerouteDeclaration(None=@n32) {name=Smoke_Bkg_Time, nodeColor=<rgba 0 0 0 1>}) {bFractionalPart=true, speedX=-0.02, speedY=0.04}) {
   samplerType=SAMPLERTYPE_LinearGrayscale
   texture=/Game/UI/Frontend/Lobby/Habanero/MiniSeasons/JethroMiniSeason/Textures/T_StormCloudDetail01_M
  }
  n34 = Saturate(None=Add(A=@n33.G) {constB=0.2})
  n35 = Multiply(A=@n31, B=AppendVector(A=Constant() {r=1.5}, B=Constant() {r=1.3}))
  n36 = NamedRerouteUsage()
  n37 = TextureSample(UVs=NamedRerouteDeclaration(None=Panner(Coordinate=@n35, Time=@n36) {bFractionalPart=true, speedX=-0.04, speedY=0.02}) {name=Smoke_Bkg_AnimUvs, nodeColor=<rgba 0 0 0 1>}) {texture=&T_UI_SmokeEffectNoise}
  NamedRerouteDeclaration(None=Multiply(A=LinearInterpolate(A=@n34, B=@n37.B), B=ScalarParameter() {defaultValue=0.8, group=Smoke1, parameterName=Smoke_OpacityMultiply})) {name=Smoke Shape, nodeColor=<rgba 0 0 0 1>}
  n38 = MaterialFunctionCall() {materialFunction=&GetUserInterfaceUV}
  n39 = Panner(Coordinate=MaterialFunctionCall(
    UVs=MaterialFunctionCall() {materialFunction=&MF_UI_UVScaling}.Scaled to Fit
    Texture Scale=Constant() {r=2}
   ) {materialFunction=&ScaleUVsByCenter}.UVs) {speedX=-0.03, speedY=0.05}
  n40 = Panner(Coordinate=MaterialFunctionCall(
    UVs=MaterialFunctionCall() {materialFunction=&MF_UI_UVScaling}.Scaled to Fit
    Texture Scale=Constant() {r=3}
   ) {materialFunction=&ScaleUVsByCenter}.UVs) {speedX=-0.02, speedY=0.04}
  n41 = Power(Base=Max(
    A=TextureSample(UVs=@n39) {texture=&T_UI_SmokeEffectNoise}.R
    B=TextureSample(UVs=@n40) {texture=&T_UI_SmokeEffectNoise}.R
   ))
  NamedRerouteDeclaration(None=Add(A=NamedRerouteDeclaration(None=@n38.Normalized UV) {name=Normalized UV, nodeColor=<rgba 0 0 0 1>}, B=Multiply(A=@n41) {constB=0.12})) {name=Distorted UV, nodeColor=<rgba 0 0 0 1>}
  n42 = MaterialFunctionCall(UVs=NamedRerouteUsage(), "Rotation Angle (0-1)"=Constant() {r=1.0167}) {materialFunction=&CustomRotator}
  NamedRerouteDeclaration(None=SmoothStep(Value=ComponentMask(None=@n42.Rotated Values) {r=true}) {constMax=0.4, constMin=0.5}) {name=ColorSeparator, nodeColor=<rgba 0 0.5781 1 1>}
  NamedRerouteDeclaration(None=VectorParameter() {parameterName=Base Color 1, defaultValue=<rgba 0.0837 0.1236 0.2448 1>}.RGB) {name=BaseColor, nodeColor=<rgba 0.8984 1 0 1>}
  NamedRerouteDeclaration(None=VectorParameter() {parameterName=Secondary Color 1, defaultValue=<rgba 0.3528 0.8352 0.8531 1>}.RGB) {name=SecondaryColor, nodeColor=<rgba 0.1016 1 0 1>}
  NamedRerouteUsage()
 ]
}
